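# Root conversation database for the Royal Angst NPC, giving its greeting and the branches offered to the player.
# One entry point leads to a Speech node whose contextual greetings react to first-ever conversation, match conversation count or specific worn outfits, with curve-driven priorities.
# The speech fans out to link nodes that jump by entry tag to shared quests and services, character dialogue, Season 34 story and Vault expertise conversations.
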ConversationDatabase /NPCLibrary/NPCs/RoyalAngst/NPCLibrary_Conversation_RoyalAngst
entry_points [{tag=Conversation.Entry.RoyalAngst, destinations [#1]}]
nodes [
 ConversationEntryPointNode_7 {
  id=#1
  class=ConversationEntryPointNode
  outputs [#2]
  properties {entryTag {tagName=Conversation.Entry.RoyalAngst}}
 }
 Speech_C_4 {
  id=#2
  class=Speech_C
  outputs [
   #3
   #4
   #5
   #6
   #7
   #8
   #9
  ]
  properties {
   nodeName=""
   generalConfig {
    defaultMessage=Don't waste my time. What do you need?
    contextualMessages [
     {
      message=Hey.
      contextRequirements [{requirement {class=FortControllerRequirement_IsFirstEverConversationWithNPC}}]
      priority {curve {rowName=Tandem.Speech.Priority.FirstEverInteraction}}
     }
     {
      message=Don't waste my time. What do you need?
      contextRequirements [{requirement {class=FortControllerRequirement_CheckNPCMatchConversationCount, properties {nPCUniqueIdTag {tagName=AISpawnerData.Type.Tandem.RoyalAngst}}}}]
      priority {curve {rowName=None}}
     }
     {
      message="The world's got a rhythm. Pay attention to it, and you can get anywhere."
      contextRequirements [{requirement {class=FortControllerRequirement_CheckNPCMatchConversationCount, properties {nPCUniqueIdTag {tagName=AISpawnerData.Type.Tandem.RoyalAngst}}}}]
      priority {curve {rowName=None}}
     }
     {
      message=I can crack any safe on this Island. Just watch me.
      contextRequirements [{requirement {class=FortControllerRequirement_CheckNPCMatchConversationCount, properties {nPCUniqueIdTag {tagName=AISpawnerData.Type.Tandem.RoyalAngst}}}}]
      priority {curve {rowName=None}}
     }
     {
      message=Cassidy Quinn. Always a sight for sore eyes.
      contextRequirements [{requirement {class=FortControllerRequirement_HasCID, properties {allowedCIds [/BRCosmetics/Athena/Items/Cosmetics/Characters/Character_CasinoReaper_Die]}}}]
      priority {curve {rowName=Tandem.Speech.Priority.OutfitReaction}}
     }
     {
      message=I like your beats. Your bars are a little... briny.
      contextRequirements [{requirement {class=FortControllerRequirement_HasCID, properties {allowedCIds [/BRCosmetics/Athena/Items/Cosmetics/Characters/Character_WartyBrine_Sus]}}}]
      priority {curve {rowName=Tandem.Speech.Priority.OutfitReaction}}
     }
     {
      message="Your security's a joke, Kane. And not even a funny one."
      contextRequirements [{requirement {class=FortControllerRequirement_HasCID, properties {allowedCIds [/BRCosmetics/Athena/Items/Cosmetics/Characters/Character_SherWolf_Gnash]}}}]
      priority {curve {rowName=Tandem.Speech.Priority.OutfitReaction}}
     }
     {
      message=Your girl Quinn came 'round. You two still runnin' together?
      contextRequirements [{requirement {class=FortControllerRequirement_HasCID, properties {allowedCIds [/BRCosmetics/Athena/Items/Cosmetics/Characters/Character_TealMink_Pelt]}}}]
      priority {curve {rowName=Tandem.Speech.Priority.OutfitReaction}}
     }
     {
      message="Skillet! How're the booms?"
      contextRequirements [{requirement {class=FortControllerRequirement_HasCID, properties {allowedCIds [/BRCosmetics/Athena/Items/Cosmetics/Characters/Character_RageDebris]}}}]
      priority {curve {rowName=Tandem.Speech.Priority.OutfitReaction}}
     }
    ]
   }
  }
 }
 ConversationLinkNode_6 {id=#9, class=ConversationLinkNode, properties {remoteEntryTag {tagName=Conversation.Entry.Expertise.Vault}}}
 ConversationLinkNode_5 {
  id=#8
  class=ConversationLinkNode
  properties {remoteEntryTag {tagName=Conversation.Entry.Season34.Expertise.Vault.Finish}}
 }
 ConversationLinkNode_4 {
  id=#7
  class=ConversationLinkNode
  properties {remoteEntryTag {tagName=Conversation.Entry.Season34.Expertise.Vault.Introduction}}
 }
 ConversationLinkNode_3 {
  id=#6
  class=ConversationLinkNode
  properties {remoteEntryTag {tagName=Conversation.Entry.RoyalAngst.Dialogue}}
 }
 ConversationLinkNode_2 {
  id=#5
  class=ConversationLinkNode
  properties {remoteEntryTag {tagName=Conversation.Entry.Season34.Story.P01.Q05}}
 }
 ConversationLinkNode_0 {id=#4, class=ConversationLinkNode, properties {remoteEntryTag {tagName=Conversation.Entry.BasicServices}}}
 ConversationLinkNode_1 {id=#3, class=ConversationLinkNode, properties {remoteEntryTag {tagName=Conversation.Entry.NPCLibrary.Quests}}}
]
